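# Material for HLOD city buildings that can show an audio-reactive rainbow equalizer in the emissive windows.
# Tintable diffuse and masked emissive form the base look, blended toward the equalizer by Equalizer_Str, which defaults to zero.
# Two custom-HLSL sine waves on quantized world-space grids, scaled by Jukebox percussion and amplitude, offset a bounds-relative height cutoff.
# That cutoff masks a panning world-projected spectrum gradient multiplied by the window texture; several experimental branches remain unconnected.
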Material /Game/Creative/Sets/HLOD_City/Material/Equalizer/CM_BuildingEqualizer
graph {
 properties {bUseMaterialAttributes=true, pixelDepthOffsetMode=PDOM_Legacy, usedShadingModels=Default Lit}
 nodes [
  n1 = AppendVector(
   A=ScalarParameter() {defaultValue=40, group=SineWave, parameterName=Rows}
   B=ScalarParameter() {defaultValue=20, group=SineWave, parameterName=Columns}
  )
  AppendVector(A=@n1, B=Constant() {r=80})
  n2 = AppendVector(
   A=ScalarParameter() {defaultValue=4, group=SineWave, parameterName=Rows}
   B=ScalarParameter() {defaultValue=20, group=SineWave, parameterName=Columns}
  )
  AppendVector(A=@n2, B=Constant() {r=80})
  n3 = Time()
  Ceil(None=@n3)
  n4 = TransformPosition(None=WorldPosition().XYZ) {transformSourceType=TRANSFORMPOSSOURCE_World}
  n5 = Divide(A=@n4, B=ScalarParameter() {defaultValue=8000, group=SineWave, parameterName=World_Scale}) {constB=4000}
  n6 = Multiply(A=ComponentMask(None=@n5) {b=true, g=true}, B=@n1)
  Ceil(None=@n6)
  Ceil(None=Time())
  n7 = Multiply(A=ComponentMask(None=@n5) {b=true, r=true}, B=@n2)
  Ceil(None=@n7)
  n8 = Multiply(A=Time(), B=ScalarParameter() {defaultValue=0.5, group=SineWave, parameterName=Time_Multi})
  Ceil(None=@n8)
  n9 = Multiply(A=Time(), B=ScalarParameter() {defaultValue=0.5, group=SineWave, parameterName=Time_Multi})
  Ceil(None=@n9)
  CollectionParameter() {collection=/Game/Athena/Events/AudioAnalysis/MPC_Audio_Reactive_Parameters, parameterName=Amplitude_Music}
  MaterialExpressionDivide_3 = Divide(A=Floor(None=@n6), B=@n1)
  ComponentMask(None=@MaterialExpressionDivide_3) {g=true, r=true}
  MaterialExpressionDivide_2 = Divide(A=Floor(None=@n7), B=@n2)
  ComponentMask(None=@MaterialExpressionDivide_2) {r=true}
  Constant() {r=1}
  Constant() {r=1}
  Constant() {r=1}
  Constant() {r=0.7}
  Constant() {r=1}
  Constant() {r=1}
  n10 = TextureCoordinate()
  n11 = MaterialFunctionCall(
   X=ScalarParameter() {defaultValue=4, group=EQ, parameterName=nbBands}
   Y=ScalarParameter() {defaultValue=10, group=EQ, parameterName=LEDsPerBand}
  ) {materialFunction=&MakeFloat2}
  MaterialExpressionDivide_0 = Divide(A=Floor(None=Multiply(A=@n10, B=@n11.Result) {constB=5}), B=@n11.Result) {constB=5}
  MaterialExpressionConstant_4 = Constant() {r=0.01}
  MaterialExpressionScalarParameter_12 = ScalarParameter() {defaultValue=8200, group=SineWave, parameterName=Frequency}
  MaterialExpressionScalarParameter_13 = ScalarParameter() {defaultValue=15, group=SineWave, parameterName=Amplitude}
  MaterialExpressionScalarParameter_14 = ScalarParameter() {defaultValue=2, group=SineWave, parameterName=Speed}
  MaterialExpressionCeil_1 = Ceil(None=Time())
  MaterialExpressionCustom_9 = Custom!(
   uv=@MaterialExpressionDivide_0
   frequency=@MaterialExpressionScalarParameter_12
   amplitude=@MaterialExpressionScalarParameter_13
   speed=@MaterialExpressionScalarParameter_14
   itime=@MaterialExpressionCeil_1
  ) {
   code="  \tfloat phase = itime * speed;\r\n \t//amplitude at position x (y value)\r\n  \tfloat yOffset = sin(uv.x * frequency + phase);\r\n    yOffset += sin(uv.x*frequency*2.1 + phase)*4.5;\r\n    //yOffset += sin(uv.x*frequency*1.72 + phase*1.121)*4.0;\r\n\t//yOffset += sin(uv.x*frequency*2.221 + phase*0.437)*5.0;\r\n\tyOffset += sin(uv.x*frequency*3.1122+ phase*4.269)*2.5;  \r\n  \treturn yOffset * amplitude;"
   description=CreateWave
   outputType=CMOT_Float1
   inputs [
    {inputName=uv, input {expression=&self:MaterialExpressionDivide_0}}
    {
     inputName=frequency
     input {expression=&self:MaterialExpressionScalarParameter_12}
    }
    {
     inputName=amplitude
     input {expression=&self:MaterialExpressionScalarParameter_13}
    }
    {
     inputName=speed
     input {expression=&self:MaterialExpressionScalarParameter_14}
    }
    {inputName=itime, input {expression=&self:MaterialExpressionCeil_1}}
   ]
  }
  Custom!(uv=@MaterialExpressionDivide_0, lineWidth=@MaterialExpressionConstant_4, wave=@MaterialExpressionCustom_9) {
   code="float result = ((uv.y-0.5) / lineWidth) + wave;    \r\nreturn (result);"
   description=DrawWave
   outputType=CMOT_Float1
   inputs [
    {inputName=uv, input {expression=&self:MaterialExpressionDivide_0}}
    {inputName=lineWidth, input {expression=&self:MaterialExpressionConstant_4}}
    {inputName=wave, input {expression=&self:MaterialExpressionCustom_9}}
   ]
  }
  Divide(A=@n5, B=MaterialFunctionCall() {materialFunction=&ObjectLocalBounds}.Local Bounds Size)
  Frac(None=@n7)
  Frac()
  Frac(None=@n6)
  Frac(None=@n8)
  Frac(None=@n9)
  n12 = TextureSampleParameter2D() {group=BuildingEqualizer, parameterName=CM_Building_Windows_Mask, texture=/Game/Creative/Sets/HLOD_City/Textures/T_CM_Building1_emissivebase}
  n13 = ScalarParameter() {defaultValue=80000, group=BuildingEqualizer, parameterName=Rainbow_Scale}
  n14 = Divide(A=WorldPosition().XYZ, B=@n13)
  n15 = Divide(A=WorldPosition().XYZ, B=@n13)
  n16 = VertexInterpolator(VS=Abs(None=VertexNormalWS()))
  n17 = MaterialFunctionCall(In=Abs(None=ComponentMask(None=@n16.PS) {r=true}), Contrast=ScalarParameter() {group=BuildingEqualizer, parameterName=Contrast_Rainbow}) {materialFunction=&CheapContrast}
  n18 = Panner(
   Coordinate=LinearInterpolate(A=ComponentMask(None=@n14) {g=true, r=true}, B=ComponentMask(None=@n15) {b=true, g=true}, Alpha=ComponentMask(None=@n17.Result) {r=true})
   Time=@n3
  ) {speedX=0.4}
  n19 = Multiply(
   A=@n12.RGB
   B=Multiply(
    A=TextureSample(UVs=@n18) {texture=/Game/Global/Textures/Default/Masks_and_Gradients/WeaponGradients/T_Wrap_Spectrum_Grad}.RGB
    B=ScalarParameter() {defaultValue=0.41, parameterName=color_str}
   )
  )
  n20 = Time()
  n21 = Saturate(None=ConstantBiasScale(None=Sine(None=@n20) {period=2}))
  n22 = Sine(None=Time()) {period=2}
  LinearInterpolate(
   A=@n19
   B=MaterialFunctionCall(Hue Shift Percentage=Multiply(A=@n12.RGB, B=@n21), Texture=@n19) {materialFunction=&HueShift}.Result
   Alpha=Saturate(None=ConstantBiasScale(None=@n22))
  )
  MaterialFunctionCall(UVs=@n18, Rotation Center=Constant2Vector() {g=0.5, r=0.5}, "Rotation Angle (0-1)"=Constant() {r=0.25}) {materialFunction=&CustomRotator}
  NamedRerouteUsage()
  ObjectPositionWS()
  n23 = CollectionParameter() {collection=/Game/Creative/Devices/Jukebox/Creative_Jukebox_MPC, desc=Potential creative fix, parameterName=Percussion}
  Reroute(None=LinearInterpolate(
    A=@n23
    B=MaterialFunctionCall() {materialFunction=/Game/Creative/Sets/PropSets/Primitives/Customization/MaterialFunctions/MF_Primitive_AudioSetup}.Result
    Alpha=ScalarParameter() {defaultValue=0.1, group=BuildingEqualizer, parameterName=CM_Building_EventAudioBlend}
   ) {constAlpha=0.25}) {desc=old setup goes into multi}
  n24 = TextureSampleParameter2D() {
   group=BuildingEqualizer
   parameterName=CM_Building_BackGround_Diffuse
   texture=/Game/Creative/Sets/HLOD_City/Textures/T_BGBuilding_City_Skyline_01_D
  }
  n25 = TextureSampleParameter2D() {
   group=BuildingEqualizer
   parameterName=CM_Building_BaseDiffuse
   texture=&T_BGBuilding_City_Skyline_01_D
  }
  n26 = TextureSampleParameter2D() {
   parameterName=CM_Building_SpecularMask
   samplerType=SAMPLERTYPE_Masks
   texture=/Game/Creative/Sets/HLOD_City/Textures/T_BGBuilding_City_Skyline_01_S
  }
  n27 = MaterialFunctionCall!(
   Transition=ScalarParameter() {defaultValue=0.5, parameterName=TintTransition, sortPriority=5}
   HM=StaticSwitchParameter(True=@n24.A, False=@n25.A) {parameterName=Use Color Swatch}
   Contrast=ScalarParameter() {defaultValue=5, parameterName=TintContrast, sortPriority=5}
   Metalness=@n26.G
   BG Color=StaticSwitchParameter(True=@n24.RGB, False=@n25.RGB) {parameterName=Use Color Swatch}
   TintColor=StaticSwitchParameter(True=@n25.RGB, False=VectorParameter() {parameterName=TintColor, sortPriority=7, defaultValue=<rgba 1 1 1 0>}.RGB) {parameterName=Use Color Swatch}
  ) {materialFunction=/Game/Athena/WinterArt/Materials/MF_ColorMask}
  n28 = StaticSwitchParameter(True=@n27.LerpMix, False=@n27.LerpMix) {parameterName=L1_UseTintColorAsDiffuse}
  n29 = TextureCoordinate() {coordinateIndex=2}
  n30 = StaticSwitchParameter(True=TextureCoordinate() {coordinateIndex=1}, False=TextureCoordinate()) {parameterName=Ch2Emissive, sortPriority=-5}
  n31 = VectorParameter() {parameterName=Emissive_Color, defaultValue=<rgba 8 5.964 0.9817 10>}
  n32 = Multiply(
   A=TextureSampleParameter2D(UVs=StaticSwitchParameter(True=@n29, False=@n30) {parameterName=Ch3Emissive, sortPriority=1}) {parameterName=CM_Building_Emissive_Mask, texture=/Game/Creative/Sets/HLOD_City/Textures/T_BGBuilding_City_Skyline_01_E}.R
   B=Multiply(A=@n31.RGB, B=@n31.RGB)
  )
  n33 = Subtract(
   A=TransformPosition(None=WorldPosition().XYZ) {transformSourceType=TRANSFORMPOSSOURCE_World}
   B=TransformPosition(None=ObjectPositionWS()) {transformSourceType=TRANSFORMPOSSOURCE_World}
  )
  n34 = MaterialFunctionCall() {materialFunction=&ObjectLocalBounds}
  n35 = LinearInterpolate(
   A=Multiply(A=@n23, B=ScalarParameter() {defaultValue=1, group=BuildingEqualizer, parameterName=CM_Building_EventAudioMultiplier})
   B=Multiply(A=@n23) {constB=2}
   Alpha=CollectionParameter() {collection=&Creative_Jukebox_MPC, desc=Potential creative fix, parameterName=AverageAmplitude}
  )
  n36 = Saturate(None=Abs(None=Min(A=@n35) {constB=50, desc=prevents crazy values}))
  MaterialExpressionConstant_9 = Constant() {r=0.01}
  MaterialExpressionScalarParameter_23 = ScalarParameter() {defaultValue=8200, group=SineWave, parameterName=Frequency}
  MaterialExpressionScalarParameter_10 = ScalarParameter() {defaultValue=15, group=SineWave, parameterName=Amplitude}
  MaterialExpressionScalarParameter_11 = ScalarParameter() {defaultValue=2, group=SineWave, parameterName=Speed}
  MaterialExpressionStaticSwitchParameter_2 = StaticSwitchParameter(True=Time(), False=@n9) {group=BuildingEqualizer, parameterName=Debug_Equalizer}
  MaterialExpressionCustom_3 = Custom!(
   uv=@MaterialExpressionDivide_3
   frequency=@MaterialExpressionScalarParameter_23
   amplitude=@MaterialExpressionScalarParameter_10
   speed=@MaterialExpressionScalarParameter_11
   itime=@MaterialExpressionStaticSwitchParameter_2
  ) {
   code="  \tfloat phase = itime * speed;\r\n \t//amplitude at position x (y value)\r\n  \tfloat yOffset = sin(uv.x * frequency + phase);\r\n    yOffset += sin(uv.x*frequency*2.1 + phase)*4.5;\r\n    //yOffset += sin(uv.x*frequency*1.72 + phase*1.121)*4.0;\r\n\t//yOffset += sin(uv.x*frequency*2.221 + phase*0.437)*5.0;\r\n\tyOffset += sin(uv.x*frequency*3.1122+ phase*4.269)*2.5;  \r\n  \treturn yOffset * amplitude;"
   description=CreateWave
   outputType=CMOT_Float1
   inputs [
    {inputName=uv, input {expression=&self:MaterialExpressionDivide_3}}
    {
     inputName=frequency
     input {expression=&self:MaterialExpressionScalarParameter_23}
    }
    {
     inputName=amplitude
     input {expression=&self:MaterialExpressionScalarParameter_10}
    }
    {
     inputName=speed
     input {expression=&self:MaterialExpressionScalarParameter_11}
    }
    {
     inputName=itime
     input {expression=&self:MaterialExpressionStaticSwitchParameter_2}
    }
   ]
  }
  n37 = Custom!(uv=@MaterialExpressionDivide_3, lineWidth=@MaterialExpressionConstant_9, wave=@MaterialExpressionCustom_3) {
   code="float result = ((uv.y-0.5) / lineWidth) + wave;    \r\nreturn (result);"
   description=DrawWave
   outputType=CMOT_Float1
   inputs [
    {inputName=uv, input {expression=&self:MaterialExpressionDivide_3}}
    {inputName=lineWidth, input {expression=&self:MaterialExpressionConstant_9}}
    {inputName=wave, input {expression=&self:MaterialExpressionCustom_3}}
   ]
  }
  MaterialExpressionConstant_7 = Constant() {r=0.01}
  MaterialExpressionScalarParameter_21 = ScalarParameter() {defaultValue=8200, group=SineWave, parameterName=Frequency}
  MaterialExpressionScalarParameter_17 = ScalarParameter() {defaultValue=15, group=SineWave, parameterName=Amplitude}
  MaterialExpressionScalarParameter_18 = ScalarParameter() {defaultValue=2, group=SineWave, parameterName=Speed}
  MaterialExpressionStaticSwitchParameter_4 = StaticSwitchParameter(True=Time(), False=@n8) {group=BuildingEqualizer, parameterName=Debug_Equalizer}
  MaterialExpressionCustom_1 = Custom!(
   uv=@MaterialExpressionDivide_2
   frequency=@MaterialExpressionScalarParameter_21
   amplitude=@MaterialExpressionScalarParameter_17
   speed=@MaterialExpressionScalarParameter_18
   itime=@MaterialExpressionStaticSwitchParameter_4
  ) {
   code="  \tfloat phase = itime * speed;\r\n \t//amplitude at position x (y value)\r\n  \tfloat yOffset = sin(uv.x * frequency + phase);\r\n    yOffset += sin(uv.x*frequency*2.1 + phase)*4.5;\r\n    //yOffset += sin(uv.x*frequency*1.72 + phase*1.121)*4.0;\r\n\t//yOffset += sin(uv.x*frequency*2.221 + phase*0.437)*5.0;\r\n\tyOffset += sin(uv.x*frequency*3.1122+ phase*4.269)*2.5;  \r\n  \treturn yOffset * amplitude;"
   description=CreateWave
   outputType=CMOT_Float1
   inputs [
    {inputName=uv, input {expression=&self:MaterialExpressionDivide_2}}
    {
     inputName=frequency
     input {expression=&self:MaterialExpressionScalarParameter_21}
    }
    {
     inputName=amplitude
     input {expression=&self:MaterialExpressionScalarParameter_17}
    }
    {
     inputName=speed
     input {expression=&self:MaterialExpressionScalarParameter_18}
    }
    {
     inputName=itime
     input {expression=&self:MaterialExpressionStaticSwitchParameter_4}
    }
   ]
  }
  n38 = Custom!(uv=@MaterialExpressionDivide_2, lineWidth=@MaterialExpressionConstant_7, wave=@MaterialExpressionCustom_1) {
   code="float result = ((uv.y-0.5) / lineWidth) + wave;    \r\nreturn (result);"
   description=DrawWave
   outputType=CMOT_Float1
   inputs [
    {inputName=uv, input {expression=&self:MaterialExpressionDivide_2}}
    {inputName=lineWidth, input {expression=&self:MaterialExpressionConstant_7}}
    {inputName=wave, input {expression=&self:MaterialExpressionCustom_1}}
   ]
  }
  n39 = Saturate(None=OneMinus(None=LinearInterpolate(A=@n37, B=@n38)))
  n40 = Subtract(
   A=Add(
    A=Divide(A=@n33, B=@n34.Local Bounds Size)
    B=ScalarParameter() {defaultValue=0.5, group=BuildingEqualizer, parameterName=CM_Building_Offset}
   )
   B=ComponentMask(None=Multiply(A=@n36, B=@n39)) {b=true, g=true, r=true}
  )
  n41 = StaticSwitchParameter(True=Subtract(A=@n40, B=@n39), False=@n40) {group=BuildingEqualizer, parameterName=Debug_Equalizer}
  n42 = ComponentMask(None=Saturate(None=Multiply(A=@n41) {constB=256})) {b=true}
  n43 = TextureSampleParameter2D() {
   parameterName=CM_Building_NormalMask
   samplerType=SAMPLERTYPE_Normal
   texture=/Game/Creative/Sets/HLOD_City/Textures/T_BGBuilding_City_Skyline_01_N
  }
  SetMaterialAttributes!(
   MaterialAttributes
   Base Color=@n28
   Metallic=@n26.G
   Specular=@n26.G
   Roughness=@n26.G
   Emissive Color=LinearInterpolate(
    A=MaterialFunctionCall!(BaseColor=@n28, EmissiveMask=@n26.A, EmissiveColor=@n32) {materialFunction=/Game/Athena/WinterArt/Materials/MF_EmissiveControls}.Result
    B=LinearInterpolate(A=@n19, Alpha=@n42) {constB=0}
    Alpha=ScalarParameter() {parameterName=Equalizer_Str}
   )
   Normal=@n43.RGB
  ) {attributeSetTypes [
    #1
    #2
    #3
    #4
    #5
    #6
   ]}
  n44 = VertexNormalWS()
  n45 = StaticComponentMaskParameter(None=VertexInterpolator(VS=Abs(None=@n44)).PS) {
   defaultG=true
   desc=Some Building want this to be the R channel due to the mesh setup
   group=SineWave
   parameterName=SineDirection
   sortPriority=35
  }
  StaticComponentMaskParameter(None=MaterialFunctionCall(In=Abs(None=@n45), Contrast=Constant() {r=100}) {materialFunction=&CheapContrast}.Result) {
   defaultG=true
   desc=Some Building want this to be the R channel due to the mesh setup
   group=SineWave
   parameterName=SineDirection
   sortPriority=35
  }
  Subtract(A=@n4, B=TransformPosition(None=ObjectPositionWS()) {transformSourceType=TRANSFORMPOSSOURCE_World})
  TextureCoordinate() {uTiling=0.6}
  TextureCoordinate() {coordinateIndex=1}
  TextureSampleParameter2D() {group=BuildingEqualizer, parameterName=CM_Building_Base_Emissive, texture=/Game/Creative/Sets/HLOD_City/Textures/T_CM_Building_1_emissive}
  TextureSample() {texture=/Game/ArtTools/MaterialLibrary/Materials/5_Color_Gradient}
  TextureSample(UVs=@n18) {texture=/Game/Creative/Sets/PropSets/Primitives/Customization/Textures/RainbowStripe/T_CM_RainbowStripe_BC2}
  TextureSample() {samplerSource=SSM_Wrap_WorldGroupSettings, texture=/Game/Characters/Fortress_CharacterShader/Textures/RainbowEffect1}
  n46 = ScalarParameter() {defaultValue=8000, parameterName=Scale}
  n47 = ComponentMask(None=Divide(A=WorldPosition().XYZ, B=@n46)) {g=true, r=true}
  n48 = ComponentMask(None=Divide(A=WorldPosition().XYZ, B=@n46)) {b=true, g=true}
  n49 = VertexNormalWS()
  n50 = ComponentMask(None=VertexInterpolator(VS=Abs(None=@n49)).PS) {r=true}
  n51 = ComponentMask(None=MaterialFunctionCall(In=Abs(None=@n50), Contrast=Constant() {r=100}) {materialFunction=&CheapContrast}.Result) {r=true}
  n52 = Time()
  TextureSample(UVs=Panner(Coordinate=LinearInterpolate(A=@n47, B=@n48, Alpha=@n51), Time=Ceil(None=@n52)) {speedX=0.2, speedY=0.2}) {texture=/Game/Creative/Sets/HLOD_City/Textures/T_CM_Building_Emissive_Mask1}
  Time()
  Time()
 ]
}
